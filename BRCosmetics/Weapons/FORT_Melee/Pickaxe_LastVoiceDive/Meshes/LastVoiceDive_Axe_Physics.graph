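# Physics asset for the LastVoiceDive two-handed pickaxe mesh, giving its dangling parts limited secondary motion.
# The handle body carries several capsules covering the shaft and head, while dyn_foot, dyn_foot_2, dyn_Batel and dyn_Bol each get one capsule.
# Constraints form two chains off the handle: dyn_foot to dyn_Batel, and dyn_foot_2 to dyn_Bol.
# Swing limits are narrow to moderate and twist is nearly locked throughout.
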PhysicsAsset /BRCosmetics/Weapons/FORT_Melee/Pickaxe_LastVoiceDive/Meshes/LastVoiceDive_Axe_Physics
bodies {
 Melee_TwoHanded_Handle [
  {shapeType=Capsule, length=88.5877, radius=13.9968, center=<1.0904 -1.2729 30.5702>, rotation=<rot -0.4474 -90 -0.1206>}
  {
   shapeType=Capsule
   length=15
   radius=6.6657
   shapeName=Melee_TwoHanded_Handle_capsule
   center=<0 -8.8267 78.2662>
   rotation=<rot 0 0 78.3233>
  }
  {
   shapeType=Capsule
   length=15
   radius=6.6657
   shapeName=Melee_TwoHanded_Handle_capsule
   center=<1.6143 7.2517 79.0393>
   rotation=<rot 0 0 106.1181>
  }
  {
   shapeType=Capsule
   length=11.6701
   radius=4.4219
   shapeName=Melee_TwoHanded_Handle_capsule
   center=<1.2556 -4.7414 92.5615>
   rotation=<rot 0 0 109.3596>
  }
 ]
 dyn_Batel [{
   shapeType=Capsule
   length=10.3403
   radius=4.7393
   shapeName=dyn_Batel_capsule
   center=<-5.3016 -8.393 0.0934>
   rotation=<rot 0 -53.7627 -86.888>
  }]
 dyn_Bol [{
   shapeType=Capsule
   length=1.324
   radius=3.5573
   shapeName=dyn_Bol_capsule
   center=<-0.9165 3.0498 0.0087>
   rotation=<rot 67.9963 -15.6598 73.1771>
  }]
 dyn_foot [{shapeType=Capsule, length=20.1746, radius=5.5648, center=<14.9551 4.5891 0.1039>, rotation=<rot 0.0484 -53.5144 -89.7033>}]
 dyn_foot_2 [{
   shapeType=Capsule
   length=14.2621
   radius=4.8833
   shapeName=dyn_foot_2_capsule
   center=<17.1689 -5.7855 -0.4689>
   rotation=<rot -0.7665 -119.3809 -82.8964>
  }]
}
constraints [
 {bone1Name=dyn_foot, bone2Name=Melee_TwoHanded_Handle, swing1LimitDegrees=9, swing2LimitDegrees=4, twistLimitDegrees=1}
 {bone1Name=dyn_Bol, bone2Name=dyn_foot_2, swing1LimitDegrees=7, swing2LimitDegrees=7, twistLimitDegrees=1}
 {bone1Name=dyn_foot_2, bone2Name=Melee_TwoHanded_Handle, swing1LimitDegrees=30, swing2LimitDegrees=30, twistLimitDegrees=2}
 {bone1Name=dyn_Batel, bone2Name=dyn_foot, swing1LimitDegrees=14, swing2LimitDegrees=3, twistLimitDegrees=1}
]
