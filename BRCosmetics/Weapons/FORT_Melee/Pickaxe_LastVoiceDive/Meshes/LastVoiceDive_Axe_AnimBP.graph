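# Animation Blueprint for the LastVoiceDive pickaxe that adds RigidBody physics secondary motion to the incoming pose.
# The simulation uses the axe's override physics asset, with dyn_pickaxe as base bone and gravity and component velocity/acceleration scales driven by variables.
# Each update sets those values for idle, moving, or moving while jumping, with by far the strongest response when moving without jumping.
# Physics alpha is reduced in the front end.
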AnimBlueprint /BRCosmetics/Weapons/FORT_Melee/Pickaxe_LastVoiceDive/Meshes/LastVoiceDive_Axe_AnimBP
target_skeleton=&/LastVoiceDive_Axe_Skeleton
graphs [
 AnimGraph {nodes [
   AnimGraphNode_SubInput_0 = Input Pose()
   n1 = RigidBody!(
    ComponentPose=Local To Component(LocalPose=@AnimGraphNode_SubInput_0)
    Alpha=Get sneaker_FE_physics_alpha() -> (sneaker_FE_physics_alpha:double)
    OverrideWorldGravity:Vector=Get Gravity() -> (Gravity:Vector)
    ExternalForce
    ComponentLinearAccScale:Vector=Get Acc scale() -> (Acc scale:Vector)
    ComponentLinearVelScale:Vector=Get Vel scale() -> (Vel scale:Vector)
   ) -> (Pose) {
    bOverrideWorldGravity=true
    overridePhysicsAsset=&/LastVoiceDive_Axe_Physics
    alphaBoolBlend {blendOption=Linear}
    baseBoneRef {boneName=dyn_pickaxe}
    componentAppliedLinearAccClamp=<1000 1000 1000>
    componentLinearAccScale=<0.18 0.18 0.18>
    componentLinearVelScale=<0.18 0.18 0.18>
    overrideWorldGravity {z=-200}
    simSpaceSettings {maxAngularAcceleration=1000, maxAngularVelocity=1000, maxLinearAcceleration=1000, maxLinearVelocity=1000, worldAlpha=0.15}
   }
   Output Pose(Result=Component To Local(ComponentPose=@n1)) {layerGroup=None}
  ]}
 EventGraph {nodes [
   Event BlueprintUpdateAnimation()
   > n1 = Sequence()
   > n2 = Branch(Condition=Get bIsMoving2D())
   > n3 = Branch(Condition=Get bIsJumping())
   > n4 = Sequence!(execute) -> (then_0, then_1, then_2:Exec)
   > Set Acc scale(Acc scale:Vector="0.5,0.5,0.5") -> (Output_Get:Vector)
   Set Vel scale(execute=@n4.then_1, Vel scale:Vector="0.5,0.5,0.5") -> (Output_Get:Vector)
   Set Gravity(execute=@n4.then_2, Gravity:Vector="0,0,-200") -> (Output_Get:Vector)
   n5 = Sequence(execute=@n3.else)
   > Set Acc scale(Acc scale="22,22,22")
   Set Vel scale(execute=@n5.then_1, Vel scale="22,22,22")
   Set Gravity(execute=@n5.then_2, Gravity="0,0,-300")
   n6 = Sequence(execute=@n2.else)
   > Set Acc scale(Acc scale="0.15,0.15,0.15")
   Set Vel scale(execute=@n6.then_1, Vel scale="0.15,0.15,0.15")
   Set Gravity(execute=@n6.then_2, Gravity="0,0,-200")
   n7 = Branch(execute=@n1.then_1, Condition=Get bIsInFrontEnd())
   > Set sneaker_FE_physics_alpha(sneaker_FE_physics_alpha:double=0.7) -> (Output_Get:double)
   Set sneaker_FE_physics_alpha(execute=@n7.else, sneaker_FE_physics_alpha=1)
   TryGetPawnOwner()
  ]}
]
variables [
 Acc scale {type=Vector}
 Gravity {type=Vector}
 sneaker_FE_physics_alpha {type=number}
 Vel scale {type=Vector}
]
